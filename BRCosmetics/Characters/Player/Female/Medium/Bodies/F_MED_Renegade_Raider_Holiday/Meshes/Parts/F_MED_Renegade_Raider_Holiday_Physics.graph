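# Physics asset for the Renegade Raider Holiday female character part, setting up simulation of the hat's dangling flaps.
# Each flap hangs from a small base sphere through limited-angle constraints, with swing1 locked on joints below the parent segment.
# The chains are asymmetric: the left uses three capsule segments, while the right uses two boxes.
# Capsules on head and neck, unconstrained to the flaps, provide body shapes for collision.
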PhysicsAsset /BRCosmetics/Characters/Player/Female/Medium/Bodies/F_MED_Renegade_Raider_Holiday/Meshes/Parts/F_MED_Renegade_Raider_Holiday_Physics
bodies {
 dyn_hat_flap_01_l [{
   shapeType=Capsule
   length=3.8219
   radius=1.0714
   shapeName=dyn_hat_flap_01_l_capsule
   center=<0.9738 -0.0932 0.0381>
   rotation=<rot 1.5228 1.0307 -82.9022>
  }]
 dyn_hat_flap_01_r [{shapeType=Box, extentX=2.8775, extentY=6.262, extentZ=1.6814, center=<1.3994 0 0.0001>}]
 dyn_hat_flap_02_l [{
   shapeType=Capsule
   length=2.0826
   radius=1.0072
   shapeName=dyn_hat_flap_02_l_capsule
   center=<1.7676 -0.0021 -0.1481>
   rotation=<rot 0 83.1278 -87.0252>
  }]
 dyn_hat_flap_base_l [{shapeType=Sphere, radius=0.5}]
 dyn_hat_flap_base_r [{shapeType=Sphere, radius=0.5}]
 dyn_hat_flap_parent_l [{shapeType=Capsule, length=1, radius=3, shapeName=dyn_hat_flap_parent_l_capsule, center=<2.5 0 0>, rotation=<rot -90 0 0>}]
 dyn_hat_flap_parent_r [{shapeType=Box, extentX=4.9191, extentY=6.808, extentZ=1.6098, center=<2.2285 0 0.2411>, rotation=<rot 10.2118 0 0>}]
 dyn_hat_offset [{shapeType=Sphere, radius=0.5}]
 head [{shapeType=Capsule, length=6.4737, radius=6.9271, rotation=<rot 90 0 0>}]
 neck_01 [
  {shapeType=Capsule, length=24.25, radius=5.75, center=<-3.629 -0.3233 0>}
  {shapeType=Capsule, length=6.9183, radius=5.75, shapeName=neck_01_capsule, center=<-3.2506 0.9654 -7.9447>}
 ]
}
constraints [
 {bone1Name=dyn_hat_flap_parent_r, bone2Name=dyn_hat_flap_base_r, swing1LimitDegrees=10, swing2LimitDegrees=20, twistLimitDegrees=3}
 {
  bone1Name=dyn_hat_flap_01_r
  bone2Name=dyn_hat_flap_parent_r
  swing1LimitDegrees=20
  swing1Motion=Locked
  swing2LimitDegrees=20
  twistLimitDegrees=3
 }
 {
  bone1Name=dyn_hat_flap_01_l
  bone2Name=dyn_hat_flap_parent_l
  swing1LimitDegrees=20
  swing1Motion=Locked
  swing2LimitDegrees=20
  twistLimitDegrees=3
 }
 {
  bone1Name=dyn_hat_flap_02_l
  bone2Name=dyn_hat_flap_01_l
  swing1LimitDegrees=20
  swing1Motion=Locked
  swing2LimitDegrees=20
  twistLimitDegrees=3
 }
 {bone1Name=dyn_hat_flap_parent_l, bone2Name=dyn_hat_flap_base_l, swing1LimitDegrees=20, swing2LimitDegrees=20, twistLimitDegrees=3}
]
